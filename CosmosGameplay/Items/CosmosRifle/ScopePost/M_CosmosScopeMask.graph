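# Unlit masked material for the Cosmos rifle scope, built around an aim-down-sights mask.
# It samples T_ADS_Mask in screen space with aspect-corrected scaling about the center, then derives smoothstepped or offset inverted masks.
# Separate vertex math re-bases the mesh in view space at an offset from the camera and shrinks it by bounds.
# Experimental branches and debug readouts remain, with no material output links shown.
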Material /CosmosGameplay/Items/CosmosRifle/ScopePost/M_CosmosScopeMask
graph {
 properties {
  bAllowFrontLayerTranslucency=false
  bAutomaticallySetUsageInEditor=false
  bCastRayTracedShadows=false
  blendMode=BLEND_Masked
  shadingModel=MSM_Unlit
  usedShadingModels=Unlit
 }
 nodes [
  Constant2Vector() {g=-1, r=1}
  Constant()
  n1 = MaterialFunctionCall() {materialFunction=&BoundingBoxBased_0-1_UVW}
  n2 = Multiply(A=Multiply(A=Subtract(A=@n1.RGB) {constB=0.5}) {constB=-1}, B=ObjectLocalBounds().Half Extents)
  n3 = ViewProperty() {property=MEVP_ViewSize}
  n4 = ComponentMask(None=@n3.Property) {g=true}
  n5 = ComponentMask(None=@n3.Property) {r=true}
  n6 = MaterialFunctionCall(Float2=Multiply(A=Divide(A=@n4, B=@n5), B=ScalarParameter() {defaultValue=1.3472, parameterName=ADS_MaskSize})) {materialFunction=&BreakOutFloat2Components}
  LinearInterpolate(
   B=Multiply(A=@n2, B=MaterialFunctionCall(Y=@n6.R, Z=@n6.R) {materialFunction=&MakeFloat3}.Result)
   Alpha=MaterialFunctionCall() {materialFunction=&DebugTimeSine}.Result
  )
  MaterialFunctionCall!(Object Basis 1, Object Basis 2, Object Basis 3, Camera Space Position Offset) {materialFunction=/Engine/Functions/Engine_MaterialFunctions02/WorldPositionOffset/AttachMeshToTheCamera}
  n7 = Frac(None=TextureCoordinate() {uTiling=3, vTiling=3})
  MaterialFunctionCall(Number=MaterialFunctionCall() {materialFunction=/Engine/Functions/Engine_MaterialFunctions02/Utility/FOV}."FOV ", UVs=@n7) {materialFunction=&DebugScalarValues}
  n8 = ViewProperty() {property=MEVP_ViewSize}
  MaterialFunctionCall!(
   Vector4=AppendVector(A=@n8.Property, B=@n8.Property)
   MaximumNumberOfDigits
   UVs=@n7
   DebugTextLocation RG_UpperRight BA_LowerLeft
   Component Spacing
  ) {materialFunction=/Engine/Functions/Engine_MaterialFunctions02/Utility/DebugFloat4Values}
  MaterialFunctionCall() {materialFunction=&ScreenResolution}
  Multiply(
   A=Subtract(A=MaterialFunctionCall() {materialFunction=&BoundingBoxBased_0-1_UVW}.RGB) {constB=0.5}
   B=Constant3Vector() {constant=<rgba 0 2 2 0>}
  ) {constB=2}
  n9 = ScalarParameter() {defaultValue=0.01, parameterName=SmoothstepMax}
  n10 = ViewProperty() {property=MEVP_ViewSize}
  n11 = AppendVector(A=Divide(A=ComponentMask(None=@n10.Property) {g=true}, B=ComponentMask(None=@n10.Property) {r=true}), B=Constant() {r=1})
  n12 = ScalarParameter() {defaultValue=1.3472, parameterName=ADS_MaskSize}
  n13 = TextureSample(UVs=MaterialFunctionCall(UVs=ScreenPosition().ViewportUV, Texture Scale=Multiply(A=@n11, B=@n12)) {materialFunction=&ScaleUVsByCenter}.UVs) {samplerType=SAMPLERTYPE_Masks, texture=/Game/Athena/HUD/reticles/Weapons/T_ADS_Mask}
  OneMinus(None=Round(None=SmoothStep(Max=@n9, Value=@n13.B) {constMax=0.3}))
  n14 = ScalarParameter() {parameterName=Mask Offset}
  Reroute(None=OneMinus(None=Add(A=@n13.B, B=@n14)))
  n15 = WorldPosition()
  n16 = MaterialFunctionCall!(
   Vector1=Constant3Vector() {constant=<rgba 0 0 -1 0>}
   Vector2=Constant3Vector() {constant=<rgba -1 0 0 0>}
   Vector3=Constant3Vector() {constant=<rgba 0 1 0 0>}
  ) {materialFunction=/Engine/Functions/Engine_MaterialFunctions02/Math/MakeVectorsOrthogonal}
  n17 = MaterialFunctionCall(
   VectorToTransform=Multiply(
    A=MaterialFunctionCall() {materialFunction=/Engine/Functions/Engine_MaterialFunctions02/WorldPositionOffset/ObjectScale}.Scale XYZ
    B=TransformPosition(None=@n15.XYZ) {transformSourceType=TRANSFORMPOSSOURCE_World, transformType=TRANSFORMPOSSOURCE_Instance}
   ) {desc=Apply Object Scale once again}
   BasisX=@n16.Vector1
   BasisY=@n16.Vector1
   BasisZ=@n16.Vector1
   BasisPosition=MaterialFunctionCall(Z=ScalarParameter() {defaultValue=250, parameterName=Attatch Offset}) {materialFunction=&MakeFloat3}.Result
  ) {materialFunction=&Transform3x3Matrix}
  Subtract(A=Add(A=Transform(None=@n17.Transformed Vector) {transformSourceType=TRANSFORMSOURCE_View}, B=CameraPositionWS()), B=WorldPosition().XYZ)
  n18 = ScalarParameter() {parameterName=Shrink, sliderMax=1, sliderMin=1}
  TransformPosition(None=Add(A=TransformPosition(None=@n15.XYZ) {transformSourceType=TRANSFORMPOSSOURCE_World}, B=Multiply(A=@n2, B=@n18)) {constB=0}) {transformType=TRANSFORMPOSSOURCE_Instance}
 ]
}
